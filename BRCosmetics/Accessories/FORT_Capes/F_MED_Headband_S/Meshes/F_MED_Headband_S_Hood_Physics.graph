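# Physics asset for the hood mesh of the F_MED_Headband_S cape cosmetic, giving capsule bodies and joints to its cape and hood bones.
# A five-bone dyn_cape chain hangs from a larger backpack_root capsule, starting with a tiny capsule at dyn_cape01.
# A three-bone hood chain branches from dyn_cape01.
# Cape joints have free swing and twist, while hood joints carry tight angle limits that keep the hood stiffer than the cape.
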PhysicsAsset /BRCosmetics/Accessories/FORT_Capes/F_MED_Headband_S/Meshes/F_MED_Headband_S_Hood_Physics
bodies {
 backpack_root [{
   shapeType=Capsule
   length=16.7056
   radius=9.6877
   shapeName=backpack_root_capsule
   center=<0.0274 -0.8001 -7.109>
   rotation=<rot 0 -90.043 -89.8103>
  }]
 dyn_cape01 [{shapeType=Capsule, length=0.505, radius=0.505, shapeName=dyn_cape01_capsule, center=<0 2 0>, rotation=<rot 0 -90 0>}]
 dyn_cape02 [{shapeType=Capsule, length=15.97, radius=4.5076, center=<0 3 0>}]
 dyn_cape03 [{shapeType=Capsule, length=15.97, radius=4.5076, center=<0 2 0>}]
 dyn_cape04 [{shapeType=Capsule, length=15.97, radius=4.5076, center=<0 2 0>}]
 dyn_cape05 [{shapeType=Capsule, length=15.97, radius=4.5076, center=<0 2 0>}]
 hood_01 [{shapeType=Capsule, length=15.97, radius=4.5076, center=<4.0879 -2.3 0>}]
 hood_02 [{shapeType=Capsule, length=15.97, radius=4.5076, center=<4.0879 -2.3 0>}]
 hood_03 [{shapeType=Capsule, length=15.97, radius=4.5076, center=<4.0879 -2.3 0>}]
}
constraints [
 {
  bone1Name=dyn_cape01
  bone2Name=backpack_root
  swing1LimitDegrees=45
  swing1Motion=Free
  swing2LimitDegrees=45
  swing2Motion=Free
  twistLimitDegrees=45
  twistMotion=Free
 }
 {bone1Name=hood_01, bone2Name=dyn_cape01, swing1LimitDegrees=15, swing2LimitDegrees=15, twistLimitDegrees=5}
 {bone1Name=hood_02, bone2Name=hood_01, swing1LimitDegrees=15, swing2LimitDegrees=15, twistLimitDegrees=5}
 {bone1Name=hood_03, bone2Name=hood_02, swing1LimitDegrees=15, swing2LimitDegrees=15, twistLimitDegrees=5}
 {
  bone1Name=dyn_cape02
  bone2Name=dyn_cape01
  swing1LimitDegrees=45
  swing1Motion=Free
  swing2LimitDegrees=45
  swing2Motion=Free
  twistLimitDegrees=45
  twistMotion=Free
 }
 {
  bone1Name=dyn_cape03
  bone2Name=dyn_cape02
  swing1LimitDegrees=45
  swing1Motion=Free
  swing2LimitDegrees=45
  swing2Motion=Free
  twistLimitDegrees=45
  twistMotion=Free
 }
 {
  bone1Name=dyn_cape04
  bone2Name=dyn_cape03
  swing1LimitDegrees=45
  swing1Motion=Free
  swing2LimitDegrees=45
  swing2Motion=Free
  twistLimitDegrees=45
  twistMotion=Free
 }
 {
  bone1Name=dyn_cape05
  bone2Name=dyn_cape04
  swing1LimitDegrees=45
  swing1Motion=Free
  swing2LimitDegrees=45
  swing2Motion=Free
  twistLimitDegrees=45
  twistMotion=Free
 }
]
